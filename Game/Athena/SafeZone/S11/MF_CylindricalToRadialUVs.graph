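# Material function in the Season 11 safe zone folder that converts cylindrical UVs into radial UVs.
# It keeps the input U and replaces V with local-space XY distance from the origin, normalized by object radius with X scale divided out.
# An optional BoundsScale input adjusts that normalization, and the distance runs per vertex through a VertexInterpolator.
# A BoundScaleFactor0 parameter is present but unconnected.
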MaterialFunction /Game/Athena/SafeZone/S11/MF_CylindricalToRadialUVs
graph {nodes [
  n1 = FunctionInput() {bUsePreviewValueAsDefault=true, inputName=CylindricalUVs, inputType=FunctionInput_Vector2}
  n2 = MaterialFunctionCall() {materialFunction=&LocalPosition}
  n3 = ObjectRadius()
  n4 = MaterialFunctionCall() {materialFunction=/Engine/Functions/Engine_MaterialFunctions02/WorldPositionOffset/ObjectScale}
  n5 = Multiply(A=Add(A=@n4.Scale X, B=@n4.Scale X)) {constB=0.5}
  n6 = FunctionInput() {bUsePreviewValueAsDefault=true, inputName=BoundsScale, inputType=FunctionInput_Scalar, previewValue=<1 0 0 1>}
  n7 = Divide(A=Distance(A=ComponentMask(None=@n2.Local Position) {g=true, r=true}, B=Constant()), B=Divide(A=Divide(A=@n3, B=@n5), B=@n6))
  FunctionOutput(None=AppendVector(A=ComponentMask(None=@n1) {r=true}, B=VertexInterpolator(VS=@n7).PS))
  ScalarParameter() {defaultValue=50, parameterName=BoundScaleFactor0}
 ]}
