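# Material function in the Athena SafeZone S11 folder that converts a 2D vector, or UVs, into radial coordinates.
# The angle comes from the cheaper Arctangent2Fast, divided by a full turn and wrapped with Frac into a 0–1 range.
# The vector length, from summed squared components, is appended as the second channel of the output.
# The input defaults to UVs remapped around a centered origin; its description warns that (0,0) is invalid.
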MaterialFunction /Game/Athena/SafeZone/S11/MF_VectorToRadialValueFast
graph {nodes [
  n1 = FunctionInput(Preview=Subtract(A=TextureCoordinate() {uTiling=2, vTiling=2}, B=Constant() {r=1})) {
   bUsePreviewValueAsDefault=true
   description="Insert a vector 2 value to retrieve the vectors angle or insert uvs to retrieve radial coordinates. Note that (0,0) is not an acceptable number.  "
   inputName=Vector or UVs
   inputType=FunctionInput_Vector2
   previewValue=<1 1 0 1>
  }
  n2 = ComponentMask(None=@n1) {g=true}
  n3 = ComponentMask(None=@n1) {r=true}
  n4 = Divide(A=ComponentMask(None=Arctangent2Fast(Y=@n2, X=@n3)) {r=true}, B=Constant() {r=6.2832})
  n5 = Multiply(A=@n1, B=@n1)
  n6 = Add(A=ComponentMask(None=@n5) {r=true}, B=ComponentMask(None=@n5) {g=true})
  FunctionOutput(None=AppendVector(A=Frac(None=@n4), B=SquareRoot(None=@n6)))
 ]}
