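# Material function in the S11 safe zone folder that generates cylindrical UVs from a mesh's local position and bounds.
# U is the fast radial angle around the local vertical axis; V is normalized height, scaled about the center by BoundsScale.
# It outputs per-pixel UVs, vertex-interpolated UVs, and a seam-fixed vertex version.
# The seam fix interpolates two half-turn-offset U values, and a Custom node uses fwidth to pick the continuous one.
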MaterialFunction /Game/Athena/SafeZone/S11/MF_CylindricalUVs
graph {nodes [
  n1 = MaterialFunctionCall() {materialFunction=&LocalPosition}
  n2 = MaterialFunctionCall() {materialFunction=&ObjectLocalBounds}
  n3 = ComponentMask(None=MaterialFunctionCall!(Vector or UVs=Divide(A=@n1."Local Position (Excluding Offsets)", B=@n2.Local Bounds Size)) {materialFunction=&/MF_VectorToRadialValueFast}.Result) {r=true}
  n4 = Divide(A=Subtract(A=@n1."Local Position (Excluding Offsets)", B=@n2.Local Bounds Minimum), B=@n2.Local Bounds Size)
  n5 = FunctionInput() {bUsePreviewValueAsDefault=true, inputName=BoundsScale, inputType=FunctionInput_Scalar, previewValue=<1 0 0 1>}
  n6 = Multiply(A=@n5) {constB=-0.5}
  n7 = ComponentMask(None=Add(A=Multiply(A=@n4, B=@n5), B=Add(A=@n6) {constB=0.5})) {b=true}
  n8 = AppendVector(A=@n3, B=@n7)
  FunctionOutput(None=@n8) {description=per-pixel uv, outputName=Non-Interpolated UV}
  MaterialExpressionVertexInterpolator_0 = VertexInterpolator(VS=Frac(None=@n3))
  n9 = Add(A=@n3) {constB=0.5}
  MaterialExpressionVertexInterpolator_1 = VertexInterpolator(VS=Subtract(A=Frac(None=@n9)) {constB=0.5})
  MaterialExpressionVertexInterpolator_2 = VertexInterpolator(VS=@n7)
  n10 = Custom!(u0=@MaterialExpressionVertexInterpolator_0.PS, u1=@MaterialExpressionVertexInterpolator_1.PS, v=@MaterialExpressionVertexInterpolator_2.PS) {
   code="float u = ( fwidth( u0 ) < fwidth( u1 )-0.001 )? u0 : u1;\r\nreturn float2(u,v);"
   description=FixPerVertexCylidricalUV
   outputType=CMOT_Float2
   inputs [
    {inputName=u0, input {expression=&self:MaterialExpressionVertexInterpolator_0}}
    {inputName=u1, input {expression=&self:MaterialExpressionVertexInterpolator_1}}
    {inputName=v, input {expression=&self:MaterialExpressionVertexInterpolator_2}}
   ]
  }
  FunctionOutput(None=@n10) {description=per-pixel uv, outputName=VertexUV_SeamFix}
  FunctionOutput(None=VertexInterpolator(VS=@n8).PS) {description=per-pixel uv, outputName=VertexUV}
 ]}
