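# Translucent unlit material for the lower cloud sheet of the safe-zone storm circle.
# Scrolling cylindrical noise offsets a height gradient that drives a two-color blend and top/bottom opacity mask, faded by PreventStormShrinking.
# MPC SafeZoneScale drives scroll speed and a world-position-offset calculation that rescales and shifts the mesh as the storm progresses.
# Separate branches compute spiral and radial noise plus spherical-harmonic wrap lighting from vertex color, without feeding that blend or mask.
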Material /Game/Athena/SafeZone/S11/M_StormCircleLowerSheet
graph {
 properties {
  bFullyRough=true
  bUsedWithInstancedStaticMeshes=true
  bUsedWithNanite=true
  blendMode=BLEND_Translucent
  opacityMaskClipValue=0.5
  pixelDepthOffsetMode=PDOM_Legacy
  shadingModel=MSM_Unlit
  translucencyLightingMode=TLM_SurfacePerPixelLighting
  usedShadingModels=Unlit
 }
 nodes [
  n1 = VertexNormalWS()
  n2 = MaterialFunctionCall() {materialFunction=&BreakOutFloat3Components}
  n3 = ScalarParameter() {group=1_WrapLighting, parameterName=NormalStrength}
  n4 = MaterialFunctionCall(X=@n2.R, Y=@n2.R, Z=Multiply(A=@n2.B, B=OneMinus(None=@n3))) {materialFunction=&MakeFloat3}
  n5 = Transform(None=MaterialFunctionCall(
    BaseNormal=Transform(None=@n1) {transformSourceType=TRANSFORMSOURCE_World, transformType=TRANSFORM_Tangent}
    AdditionalNormal=Normalize(VectorInput=@n4.Result)
   ) {materialFunction=&BlendAngleCorrectedNormals}.Result)
  Add(A=Multiply(A=@n5) {constB=0.5}) {constB=0.5}
  CollectionParameter() {
   collection=&FortniteMaterialParameters
   parameterName=SunlightDirectionalVector
  }
  n6 = ScalarParameter() {defaultValue=1, parameterName=BoundsScale}
  n7 = MaterialFunctionCall!(BoundsScale=@n6) {materialFunction=&/MF_CylindricalUVs}
  ComponentMask(None=@n7.VertexUV_SeamFix) {g=true}
  n8 = CollectionParameter() {
   collection=&FortniteMaterialParameters
   parameterName=SunlightDirectionalVector
  }
  ComponentMask(None=@n8) {b=true, g=true, r=true}
  Constant() {r=1}
  n9 = VertexNormalWS()
  n10 = Constant3Vector() {constant=<rgba 0 0 1 0>}
  n11 = Power(Base=Abs(None=DotProduct(A=@n9, B=@n10)), Exp=ScalarParameter() {group=Layer0_TopUndersideRadial, parameterName=VerticalAttenuation})
  n12 = MaterialFunctionCall(
   X=Round(None=ScalarParameter() {defaultValue=1, group=Layer1_TopCylindrical, parameterName=Layer1CylindricalUVScaleU})
   Y=ScalarParameter() {defaultValue=1, group=Layer1_TopCylindrical, parameterName=Layer1CylindricalUVScaleV}
  ) {materialFunction=&MakeFloat2}
  n13 = LinearInterpolate(
   A=ScalarParameter() {defaultValue=1, group=StormProgression, parameterName=GlobalSpeedEnd}
   B=ScalarParameter() {defaultValue=1, group=StormProgression, parameterName=GlobalSpeedStart}
   Alpha=Power(
    Base=CollectionParameter() {collection=/Game/Athena/SafeZone/MPC_SafeZoneClouds, parameterName=SafeZoneScale}
    Exp=ScalarParameter() {defaultValue=0.5, group=StormProgression, parameterName=GlobalSpeedDamping}
   )
  )
  n14 = Multiply(A=Time(), B=@n13)
  n15 = AppendVector(
   A=ScalarParameter() {group=Layer1_TopCylindrical, parameterName=HorizontalScrollingSpeed}
   B=ScalarParameter() {group=Layer1_TopCylindrical, parameterName=VerticalScrollingSpeed}
  )
  n16 = TextureSampleParameter2D(UVs=Add(A=Multiply(A=@n7.VertexUV_SeamFix, B=@n12.Result), B=Multiply(A=@n14, B=@n15))) {
   group=Layer1_TopCylindrical
   parameterName=L1_NoiseTexture0
   samplerType=SAMPLERTYPE_LinearColor
   texture=&/T_TilingCloudsN_01
  }
  n17 = AppendVector(A=Subtract(A=Multiply(A=@n16.RGB) {constB=2}), B=@n16.A)
  n18 = Round(None=ScalarParameter() {defaultValue=1, group=Layer1_TopCylindrical, parameterName=Layer1DetailCylindricalUVScaleU})
  n19 = ScalarParameter() {defaultValue=1, group=Layer1_TopCylindrical, parameterName=Layer1DetailCylindricalUVScaleV}
  n20 = MaterialFunctionCall(Float2=Multiply(
    A=@n7.VertexUV_SeamFix
    B=MaterialFunctionCall(X=@n18, Y=@n19) {materialFunction=&MakeFloat2}.Result
   )) {materialFunction=&BreakOutFloat2Components}
  n21 = Multiply(A=Multiply(A=Time(), B=@n13), B=ScalarParameter() {group=Layer1_TopCylindrical, parameterName=Layer1VerticalScrollingSpeed})
  n22 = TextureSampleParameter2D(UVs=AppendVector(A=@n20.R, B=Add(A=@n20.G, B=@n21))) {
   group=Layer1_TopCylindrical
   parameterName=L1_NoiseTexture1
   samplerType=SAMPLERTYPE_LinearColor
   texture=&T_TilingCloudsN_01
  }
  n23 = ScalarParameter() {defaultValue=1, group=Layer1_TopCylindrical, parameterName=Layer1SubNoisePow}
  n24 = AppendVector(
   A=Subtract(A=Multiply(A=@n22.RGB) {constB=2})
   B=Multiply(A=Power(Base=@n22.A, Exp=@n23), B=ScalarParameter() {defaultValue=1, group=Layer1_TopCylindrical, parameterName=Layer1SubNoiseMul})
  )
  n25 = ComponentMask(None=Multiply(A=OneMinus(None=@n11), B=Multiply(A=@n17, B=@n24))) {a=true}
  n26 = Add(A=ComponentMask(None=@n7.VertexUV_SeamFix) {g=true}, B=Multiply(A=@n25, B=ScalarParameter() {parameterName=noiseoffset}))
  LinearInterpolate(A=VectorParameter() {parameterName=LowerSheetCloudsColor}.RGB, B=VectorParameter() {parameterName=LowerSheetCloudsColor2}.RGB, Alpha=@n26)
  n27 = Add(A=@n26, B=ScalarParameter() {group=Mask, parameterName=MaskGradientBottomOffset})
  n28 = ScalarParameter() {group=Mask, parameterName=MaskGradientBottomMul}
  n29 = Power(Base=Saturate(None=Multiply(A=@n27, B=@n28)), Exp=ScalarParameter() {defaultValue=1, parameterName=MaskGradientPow})
  n30 = Multiply(
   A=Add(A=@n26, B=ScalarParameter() {group=Mask, parameterName=MaskGradientTopOffset})
   B=ScalarParameter() {defaultValue=1, group=Mask, parameterName=MaskGradientTopMul}
  )
  n31 = Power(Base=OneMinus(None=Saturate(None=@n30)), Exp=ScalarParameter() {defaultValue=1, parameterName=MaskGradientTopPow})
  n32 = ScalarParameter() {parameterName=CloudMinOpacity}
  LinearInterpolate(
   A=Min(A=Min(A=@n29, B=@n31), B=Add(A=@n25, B=@n32))
   Alpha=CollectionParameter() {collection=&MPC_SafeZoneClouds, parameterName=PreventStormShrinking}
  ) {constAlpha=0, constB=0}
  n33 = CollectionParameter() {collection=&MPC_SafeZoneClouds, parameterName=SafeZoneScale}
  n34 = ScalarParameter() {defaultValue=0.3, group=StormProgression, parameterName=BottomScaleDamping}
  n35 = Subtract(
   A=MaterialFunctionCall!(VectorToTransform=WorldPosition().XYZ, BasisX, BasisY, BasisZ) {materialFunction=/Engine/Functions/Engine_MaterialFunctions02/Math/InverseTransformMatrix}.Transformed Vector 3
   B=ObjectPositionWS()
  )
  n36 = Multiply(
   A=LinearInterpolate(
    A=ScalarParameter() {defaultValue=30, group=WPO, parameterName=BottomScaleEnd}
    B=ScalarParameter() {defaultValue=30, group=WPO, parameterName=BottomScaleStart}
    Alpha=Power(Base=@n33, Exp=@n34)
   )
   B=@n35
  )
  n37 = MaterialFunctionCall() {materialFunction=&LocalPosition}
  n38 = MaterialFunctionCall() {materialFunction=&ObjectLocalBounds}
  n39 = Multiply(A=Divide(A=Subtract(A=@n37."Local Position (Excluding Offsets)", B=@n38.Local Bounds Minimum), B=@n38.Local Bounds Size), B=@n6)
  n40 = Add(A=Multiply(A=@n6) {constB=-0.5}) {constB=0.5}
  n41 = Saturate(None=ComponentMask(None=Add(A=@n39, B=@n40)) {b=true})
  n42 = Power(Base=OneMinus(None=@n41), Exp=ScalarParameter() {defaultValue=20, group=WPO, parameterName=BottomScalePower})
  Multiply(A=@n36, B=@n42)
  n43 = Round(None=ScalarParameter() {defaultValue=1, group=Layer2_Spiral, parameterName=SpiralUVScaleU})
  n44 = CollectionParameter() {collection=&MPC_SafeZoneClouds, parameterName=SafeZoneScale}
  n45 = LinearInterpolate(
   A=ScalarParameter() {defaultValue=1, group=StormProgression, parameterName=SpiralVscaleEnd}
   B=ScalarParameter() {defaultValue=0.7, group=StormProgression, parameterName=SpiralVscaleStart}
   Alpha=@n44
  )
  n46 = Multiply(A=ScalarParameter() {defaultValue=1, group=Layer2_Spiral, parameterName=SpiralUVScaleV}, B=@n45)
  n47 = MaterialFunctionCall(Float2=Multiply(
    A=@n7.VertexUV_SeamFix
    B=MaterialFunctionCall(X=@n43, Y=@n46) {materialFunction=&MakeFloat2}.Result
   )) {materialFunction=&BreakOutFloat2Components}
  n48 = ScalarParameter() {defaultValue=-8, group=Layer2_Spiral, parameterName=SpiralIntensity}
  n49 = LinearInterpolate(
   A=ScalarParameter() {defaultValue=1, group=StormProgression, parameterName=GlobalSpiralIntensityEnd}
   B=ScalarParameter() {defaultValue=1, group=StormProgression, parameterName=GlobalSpiralIntensityStart}
   Alpha=Power(Base=@n44, Exp=ScalarParameter() {defaultValue=0.3, group=StormProgression, parameterName=SpiralScaleDamping})
  )
  n50 = ScalarParameter() {defaultValue=0.5, group=Layer2_Spiral, parameterName=SpiralAttenuationHeight}
  n51 = Power(
   Base=Saturate(None=Distance(A=@n47.G, B=@n50))
   Exp=ScalarParameter() {defaultValue=0.2, group=Layer2_Spiral, parameterName=SpiralAttenuationPow}
  )
  n52 = Multiply(A=Multiply(A=@n47.G, B=Multiply(A=@n48, B=@n49)) {constB=0.1}, B=@n51)
  n53 = Multiply(A=Time(), B=@n13)
  n54 = Multiply(A=@n53, B=ScalarParameter() {group=Layer2_Spiral, parameterName=SpiralSpeed})
  n55 = TextureSampleParameter2D(UVs=MaterialFunctionCall(X=Add(A=@n47.R, B=@n52), Y=Add(A=@n47.G, B=@n54)) {materialFunction=&MakeFloat2}.Result) {
   group=Layer2_Spiral
   parameterName=L2NoiseTexture0
   samplerType=SAMPLERTYPE_LinearColor
   texture=&T_TilingCloudsN_01
  }
  n56 = Multiply(A=@n55.RGB) {constB=2}
  n57 = Round(None=ScalarParameter() {defaultValue=1, group=Layer2_Spiral, parameterName=DetailSpiralUVScaleU})
  n58 = Multiply(A=ScalarParameter() {defaultValue=1, group=Layer2_Spiral, parameterName=DetailSpiralUVScaleV}, B=@n45)
  n59 = MaterialFunctionCall(Float2=Multiply(
    A=@n7.VertexUV_SeamFix
    B=MaterialFunctionCall(X=@n57, Y=@n58) {materialFunction=&MakeFloat2}.Result
   )) {materialFunction=&BreakOutFloat2Components}
  n60 = ScalarParameter() {defaultValue=-8, group=Layer2_Spiral, parameterName=DetailSpiralIntensity}
  n61 = Multiply(A=Multiply(A=@n59.G, B=Multiply(A=@n60, B=@n49)) {constB=0.1}, B=@n51)
  n62 = Multiply(A=@n53, B=ScalarParameter() {group=Layer2_Spiral, parameterName=DetailSpiralSpeed})
  n63 = TextureSampleParameter2D(UVs=MaterialFunctionCall(X=Add(A=@n59.R, B=@n61), Y=Add(A=@n59.G, B=@n62)) {materialFunction=&MakeFloat2}.Result) {
   group=Layer2_Spiral
   parameterName=L2NoiseTexture1
   samplerType=SAMPLERTYPE_LinearColor
   texture=&T_TilingCloudsN_01
  }
  n64 = Multiply(A=@n63.RGB) {constB=2}
  n65 = ScalarParameter() {defaultValue=1, group=Layer2_Spiral, parameterName=SubNoisePow}
  Multiply(A=AppendVector(A=Subtract(A=@n56), B=@n55.A), B=AppendVector(A=Subtract(A=@n64), B=Power(Base=@n63.A, Exp=@n65)))
  n66 = MaterialFunctionCall!(BoundsScale=@n6, CylindricalUVs=@n7.VertexUV_SeamFix) {materialFunction=&/MF_CylindricalToRadialUVs}
  n67 = Multiply(
   A=Multiply(A=Time(), B=@n13)
   B=AppendVector(
    A=ScalarParameter() {group=Layer0_TopUndersideRadial, parameterName=SpeedU}
    B=ScalarParameter() {group=Layer0_TopUndersideRadial, parameterName=SpeedV}
   )
  )
  n68 = ScalarParameter() {defaultValue=10, group=Layer0_TopUndersideRadial, parameterName=ScaleU}
  n69 = ScalarParameter() {defaultValue=10, group=Layer0_TopUndersideRadial, parameterName=ScaleV}
  n70 = TextureSampleParameter2D(UVs=Multiply(A=Add(A=@n66.Result, B=@n67), B=AppendVector(A=@n68, B=@n69))) {
   group=Layer0_TopUndersideRadial
   parameterName=L0NoiseTexture0
   samplerType=SAMPLERTYPE_LinearColor
   texture=&T_TilingCloudsN_01
  }
  n71 = Multiply(A=@n70.RGB) {constB=2}
  n72 = Power(Base=@n70.A, Exp=ScalarParameter() {defaultValue=1, group=Layer0_TopUndersideRadial, parameterName=Layer0_pow})
  n73 = ScalarParameter() {defaultValue=1, group=Layer0_TopUndersideRadial, parameterName=Layer0_mul}
  Multiply(A=@n11, B=AppendVector(A=Subtract(A=@n71), B=Multiply(A=@n72, B=@n73)))
  n74 = LinearInterpolate(
   A=ScalarParameter() {defaultValue=1, group=StormProgression, parameterName=TopScaleEnd}
   B=ScalarParameter() {group=StormProgression, parameterName=TopScaleStart}
   Alpha=Power(
    Base=CollectionParameter() {collection=&MPC_SafeZoneClouds, parameterName=SafeZoneScale}
    Exp=ScalarParameter() {defaultValue=0.3, group=StormProgression, parameterName=TopScaleDamping}
   )
  )
  n75 = ComponentMask(None=LinearInterpolate(A=Multiply(A=@n35, B=@n74), B=@n36, Alpha=@n42)) {g=true, r=true}
  n76 = WorldPosition()
  n77 = CollectionParameter() {collection=&MPC_SafeZoneClouds, parameterName=SafeZoneScale}
  n78 = ScalarParameter() {defaultValue=0.3, group=StormProgression, parameterName=ScaleDamping}
  n79 = Multiply(
   A=LinearInterpolate(
    A=VectorParameter() {group=StormProgression, parameterName=GlobalScaleEnd, sortPriority=1, defaultValue=<rgba 1 1 1 0>}.RGB
    B=VectorParameter() {group=StormProgression, parameterName=GlobalScaleStart, defaultValue=<rgba 1 1 5 0>}.RGB
    Alpha=Power(Base=@n77, Exp=@n78)
   )
   B=OneMinus(None=CollectionParameter() {collection=&MPC_SafeZoneClouds, parameterName=CloudStartingBias})
  )
  n80 = ComponentMask(None=Subtract(A=Multiply(A=@n76.XYZ, B=@n79), B=@n76.XYZ)) {b=true}
  n81 = VectorParameter() {group=StormProgression, parameterName=GlobalPosEnd, sortPriority=3, defaultValue=<rgba 0 0 -20000 0>}
  n82 = VectorParameter() {group=StormProgression, parameterName=GlobalPosStart, sortPriority=2, defaultValue=<rgba 0 0 -50000 0>}
  n83 = LinearInterpolate(
   A=@n81.RGB
   B=@n82.RGB
   Alpha=Power(
    Base=CollectionParameter() {collection=&MPC_SafeZoneClouds, parameterName=SafeZoneScale}
    Exp=ScalarParameter() {defaultValue=0.3, group=StormProgression, parameterName=PosDamping}
   )
  )
  n84 = Saturate(None=@n41)
  Multiply(A=Add(A=AppendVector(A=@n75, B=@n80), B=Multiply(A=@n83, B=@n84)), B=ScalarParameter() {group=WPO, parameterName=EnableWPO})
  n85 = MaterialFunctionCall() {materialFunction=&ObjectLocalBounds}
  n86 = Subtract(
   A=MaterialFunctionCall() {materialFunction=&LocalPosition}."Local Position (Excluding Offsets)"
   B=@n85.Local Bounds Minimum
  )
  Saturate(None=ComponentMask(None=Divide(A=@n86, B=@n85.Local Bounds Size)) {b=true})
  n87 = VertexColor()
  n88 = AppendVector(A=@n87, B=@n87)
  n89 = TextureSampleParameter2D!(UVs, Level=ScalarParameter() {group=SH_Depth, parameterName=SH_Blur}, Apply View MipBias) {mipValueMode=TMVM_MipLevel, parameterName=SH_Map, texture=null}
  StaticSwitchParameter(True=@n88, False=AppendVector(A=@n89.RGB, B=@n89.RGB)) {defaultValue=true, group=SH_Depth, parameterName=SH_From_VC}
  n90 = MaterialFunctionCall(Float4=@n88) {materialFunction=&BreakOutFloat4Components}
  n91 = Multiply(A=@n90.R, B=OneMinus())
  n92 = VectorParameter() {group=SH_Depth, parameterName=SH_RemapHigh, defaultValue=<rgba 1.6 0.5 0.5 0.5>}
  n93 = AppendVector(A=@n92.RGB, B=@n92.RGB)
  n94 = VectorParameter() {group=SH_Depth, parameterName=SH_RemapLow, defaultValue=<rgba 1.2 -0.5 -0.5 -0.5>}
  n95 = AppendVector(A=@n94.RGB, B=@n94.RGB)
  MaterialExpressionAdd_23 = Add(
   A=Multiply(
    A=MaterialFunctionCall(X=@n91, Y=@n90.G, Z=@n90.G, A=@n90.G) {materialFunction=&MakeFloat4}.Result
    B=Subtract(A=@n93, B=@n95)
   )
   B=@n95
  )
  n96 = MaterialFunctionCall(Float3=Multiply(A=Transform() {transformSourceType=TRANSFORMSOURCE_World, transformType=TRANSFORM_Local}, B=Constant3Vector() {constant=<rgba 1 1 -1 0>}) {constB=-1}) {materialFunction=&BreakOutFloat3Components}
  MaterialExpressionMaterialFunctionCall_51 = MaterialFunctionCall(X=@n96.R, Y=@n96.R, Z=@n96.R) {materialFunction=&MakeFloat3}
  n97 = Custom!(SH=@MaterialExpressionAdd_23, Direction=@MaterialExpressionMaterialFunctionCall_51.Result) {
   code="//SH decode,  Direction = LightDir, SH = 2band coefficients. (encoded in Y-up)\r\nDirection = normalize(Direction);\r\nfloat4 shEvalDirection;\r\nshEvalDirection.x = 0.282095;\r\nshEvalDirection.y = -0.488603 * Direction.y;\r\nshEvalDirection.z = 0.488603 * Direction.z;\r\nshEvalDirection.w = -0.488603 * Direction.x;\r\nreturn dot(SH, shEvalDirection);"
   description=SphericalHarmonicsEval
   outputType=CMOT_Float1
   inputs [
    {inputName=SH, input {expression=&self:MaterialExpressionAdd_23}}
    {inputName=Direction, input {expression=&self:MaterialExpressionMaterialFunctionCall_51}}
   ]
  }
  n98 = ScalarParameter() {defaultValue=5, group=1_WrapLighting, parameterName=SHWrapTermPow}
  n99 = Saturate(None=Multiply(A=Power(Base=@n97, Exp=@n98) {constExponent=5}, B=ScalarParameter() {defaultValue=6, group=1_WrapLighting, parameterName=SHWrapTermMul}) {constB=2})
  n100 = DotProduct(A=@n5)
  n101 = Add(A=Multiply(A=Max(A=@n100) {constB=0}, B=@n99), B=OneMinus(None=@n99))
  n102 = Power(
   Base=ComponentMask(None=@n7.VertexUV_SeamFix) {g=true}
   Exp=ScalarParameter() {defaultValue=1, group=1_WrapLighting, parameterName=WrapLightingVerticalAttenuationContrast}
  )
  n103 = ScalarParameter() {defaultValue=1, group=1_WrapLighting, parameterName=WrapLightingVerticalAttenuationIntensity}
  n104 = Multiply(A=Multiply(A=Power(Base=@n101)), B=Saturate(None=Multiply(A=@n102, B=@n103)))
  StaticSwitchParameter(True=StaticSwitchParameter(True=@n99) {parameterName=DebugSHDiffuseWrapOrBackScattering}, False=Add(A=Multiply(B=@n104))) {parameterName=DebugSHTerms}
  Subtract(A=Multiply() {constB=2})
  VertexInterpolator(VS=@n97)
 ]
}
